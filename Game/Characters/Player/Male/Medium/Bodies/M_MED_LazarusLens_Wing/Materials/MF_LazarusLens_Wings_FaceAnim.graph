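# Material function that animates tattoo lines on the LazarusLens Wing character's head, outputting a tattoo base color and mask.
# Procedural distance fields around two offset points, with time, noise and mouth-gradient controls, form repeating lines that a smooth-min merges with a static head SDF texture.
# FX texture channels and pre-skinned-position sphere masks confine the lines on the face.
# Time-driven, ear-aware radial rings add pulsing cyan waves.
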MaterialFunction /Game/Characters/Player/Male/Medium/Bodies/M_MED_LazarusLens_Wing/Materials/MF_LazarusLens_Wings_FaceAnim
graph {nodes [
  n1 = TextureSample() {
   samplerType=SAMPLERTYPE_LinearGrayscale
   texture=/Game/Characters/Player/Male/Medium/Bodies/M_MED_LazarusLens_Wing/Textures/T_LazarusLens_Wing_Head_SDF
  }
  n2 = Multiply(
   A=LinearInterpolate(
    A=ScalarParameter() {defaultValue=8, parameterName=ThicknessModif_PowerMin}
    B=ScalarParameter() {defaultValue=32, parameterName=ThicknessModif_PowerMax}
    Alpha=NamedRerouteUsage()
   ) {constA=0.1}
   B=OneMinus(None=@n1.R)
  ) {constB=8}
  n3 = ScalarParameter() {defaultValue=1, parameterName=LineDistance}
  n4 = Add(A=Divide(A=NamedRerouteUsage(), B=@n3)) {constB=0.5}
  n5 = Abs(None=Subtract(A=Frac(None=@n4)) {constB=0.5})
  n6 = Multiply(
   A=MaterialFunctionCall(In=Multiply(A=@n5, B=@n3), Contrast=NamedRerouteUsage()) {materialFunction=&CheapContrast}.Result
   B=ScalarParameter() {defaultValue=6, parameterName=LinesMult}
  ) {constB=2}
  n7 = TextureCoordinate()
  n8 = AppendVector(A=ScalarParameter() {defaultValue=1.1, parameterName=AsymRadiusY}, B=Constant() {r=1})
  n9 = ScalarParameter() {defaultValue=0.39, parameterName=AsymPosX}
  n10 = Constant() {r=0.5}
  n11 = OneMinus(None=SphereMask(
    A=MaterialFunctionCall(UVs=@n7, Texture Scale=@n8) {materialFunction=&ScaleUVsByCenter}.UVs
    B=AppendVector(A=@n9, B=@n10)
    Radius=ScalarParameter() {defaultValue=0.1, parameterName=AsymRadiusX}
   ) {attenuationRadius=0.1, hardnessPercent=0})
  n12 = SmoothStep(
   Min=ScalarParameter() {defaultValue=0.3, parameterName=ThicknessSmoothStepMin}
   Max=ScalarParameter() {defaultValue=1, parameterName=ThicknessSmoothStepMax}
   Value=@n11
  ) {constMin=0.5}
  n13 = OneMinus(None=Saturate(None=Multiply(A=@n6, B=@n12)))
  n14 = ScalarParameter() {defaultValue=2, parameterName=SmoothMinK}
  Composite!(A=@n2, B=@n13, K=@n14) {subgraphName=SmoothMin}
  Composite!() {subgraphName=EarGradient}
  Composite() {subgraphName=TopTaper}
  Constant2Vector() {g=1, r=0.5}
  n15 = TextureCoordinate()
  n16 = Constant() {r=0.125}
  n17 = Step(Y=@n15) {constX=0.5}
  n18 = MaterialFunctionCall(UVs=@n15, "Rotation Angle (0-1)"=LinearInterpolate(A=@n16, B=Multiply(A=@n16) {constB=-1}, Alpha=ComponentMask(None=@n17) {g=true})) {materialFunction=&CustomRotator}
  n19 = Abs(None=ConstantBiasScale(None=ComponentMask(None=@n18.Rotated Values) {g=true}) {bias=-0.5, scale=2})
  n20 = SmoothStep(Value=@n19) {constMax=0.8, constMin=0.3}
  MaterialExpressionReroute_22 = Reroute(None=@n20)
  n21 = MaterialFunctionCall(
   UVs=TextureCoordinate()
   Texture Scale=AppendVector(A=LinearInterpolate(Alpha=@MaterialExpressionReroute_22.Output2) {constA=0.5, constB=0.33}, B=Constant() {r=1})
  ) {materialFunction=&ScaleUVsByCenter}
  n22 = Constant2Vector() {g=0.5, r=0.5}
  n23 = Multiply(A=Time()) {constB=0.066}
  n24 = Sine(None=Multiply(A=SphereMask(A=@n21.UVs, B=@n22, Radius=@n23) {attenuationRadius=0.5, hardnessPercent=0}, B=@n23)) {desc=.4, period=0.045}
  ConstantBiasScale(None=@n24)
  FunctionOutput(None=NamedRerouteUsage()) {outputName=TatooBaseColor}
  FunctionOutput(None=NamedRerouteUsage()) {outputName=TatooMask}
  n25 = ScalarParameter() {defaultValue=0.15, desc=0.2, parameterName=NoiseStrength}
  n26 = ScalarParameter() {defaultValue=0.85, parameterName=NoiseAroundStaticLinesPower}
  n27 = ScalarParameter() {defaultValue=1, parameterName=NoiseAroundStaticLinesPower2}
  n28 = NamedRerouteUsage()
  Multiply(A=@n25, B=OneMinus(None=SmoothStep(Min=@n26, Max=@n27, Value=@n28) {constMax=0.9, constMin=0.9}))
  n29 = TextureSample() {
   samplerType=SAMPLERTYPE_LinearColor
   texture=/Game/Characters/Player/Male/Medium/Bodies/M_MED_LazarusLens_Wing/Textures/T_LazarusLens_Wing_Head_FX
  }
  n30 = Multiply(A=NamedRerouteUsage(), B=Multiply(A=@n29.B, B=NamedRerouteUsage()))
  n31 = NamedRerouteUsage()
  n32 = Constant3Vector() {constant=<rgba 0 15 165 0>}
  n33 = Max(A=@n30, B=Multiply(A=@n29.G, B=SphereMask(A=@n31, B=@n32) {attenuationRadius=24, hardnessPercent=90}))
  n34 = LinearInterpolate(A=NamedRerouteUsage(), B=VectorParameter() {parameterName=LinesColor, defaultValue=<rgba 0.1 0.098 0.098 1>}.RGB, Alpha=@n33) {constA=0.18}
  n35 = NamedRerouteUsage()
  n36 = NamedRerouteUsage()
  n37 = Multiply(A=@n31, B=Constant3Vector() {constant=<rgba 1 1 0.5 0>})
  n38 = Constant3Vector() {constant=<rgba 0 5 78 0>}
  n39 = Max(A=@n33, B=OneMinus(None=SphereMask(A=@n37, B=@n38) {attenuationRadius=9, hardnessPercent=80}))
  NamedRerouteDeclaration(None=LinearInterpolate(A=@n34, B=LinearInterpolate(A=@n34, B=@n35, Alpha=@n36), Alpha=Multiply(A=@n29.R, B=@n39))) {name=Face, nodeColor=<rgba 0.1719 0 1 1>}
  NamedRerouteDeclaration(None=FunctionInput() {inputName=PreSkinLocalPos, sortPriority=3}) {name=PreSkinLoc, nodeColor=<rgba 0 0.5312 1 1>}
  NamedRerouteDeclaration(None=FunctionInput() {inputName=Emissive, sortPriority=2}) {name=InEmissive, nodeColor=<rgba 0.9922 1 0 1>}
  NamedRerouteDeclaration(None=MaterialFunctionCall(X=ScalarParameter() {defaultValue=2, parameterName=UVScaleX}, Y=ScalarParameter() {defaultValue=6, parameterName=UVScaleY}) {materialFunction=&MakeFloat2}.Result) {name=UVScale, nodeColor=<rgba 1 0 0.3984 1>}
  NamedRerouteDeclaration(None=Multiply(A=ScalarParameter() {defaultValue=-0.5, parameterName=Speed}, B=Time())) {name=AnimatedOffset, nodeColor=<rgba 1 0 0.2109 1>}
  n40 = Add(A=NamedRerouteUsage(), B=Multiply(A=NamedRerouteUsage(), B=NamedRerouteUsage()) {constB=10})
  n41 = NamedRerouteUsage()
  n42 = Add(A=NamedRerouteUsage(), B=Multiply(A=NamedRerouteUsage(), B=NamedRerouteUsage()) {constB=10})
  n43 = NamedRerouteUsage()
  NamedRerouteDeclaration(None=LinearInterpolate(A=Add(A=@n40, B=@n41), B=Add(A=@n42, B=@n43), Alpha=NamedRerouteUsage())) {name=SDF, nodeColor=<rgba 0.7109 0 1 1>}
  n44 = NamedRerouteUsage()
  n45 = Multiply(A=NamedRerouteUsage(), B=Multiply(A=Sine(None=@n44)) {constB=0.1}) {constB=0.2}
  n46 = Add(A=NamedRerouteUsage(), B=@n45)
  n47 = Add(A=NamedRerouteUsage(), B=@n45)
  n48 = NamedRerouteUsage()
  NamedRerouteDeclaration(None=MaterialFunctionCall(In=LinearInterpolate(A=@n46, B=@n47, Alpha=@n48), Contrast=ScalarParameter() {defaultValue=128, parameterName=SDFMaskThreshold}) {materialFunction=&CheapContrast}.Result) {name=SDFMask, nodeColor=<rgba 1 0.2812 0 1>}
  MaterialExpressionReroute_10 = Reroute(None=@n13)
  MaterialExpressionReroute_17 = Reroute(None=@n2)
  n49 = Multiply(A=Subtract(A=@MaterialExpressionReroute_10.B, B=@MaterialExpressionReroute_17.A)) {constB=0.5}
  MaterialExpressionReroute_18 = Reroute(None=@n14)
  n50 = Saturate(None=Add(A=Divide(A=@n49, B=@MaterialExpressionReroute_18.K)) {constB=0.5})
  n51 = Subtract(
   A=LinearInterpolate(A=@MaterialExpressionReroute_10.B, B=@MaterialExpressionReroute_17.A, Alpha=@n50)
   B=Multiply(A=OneMinus(None=@n50), B=Multiply(A=@n50, B=@MaterialExpressionReroute_18.K))
  )
  MaterialExpressionReroute_16 = Reroute(None=@n51)
  n52 = Saturate(None=@MaterialExpressionReroute_16.Output)
  NamedRerouteDeclaration(None=Power(Base=OneMinus(None=@n52)) {constExponent=100}) {name=Lines, nodeColor=<rgba 0 1 0.9766 1>}
  n53 = ScalarParameter() {parameterName=UVoffsetX}
  n54 = ScalarParameter() {defaultValue=0.03, parameterName=UVoffsetY}
  NamedRerouteDeclaration(None=MaterialFunctionCall!(
    UV=TextureCoordinate()
    Offset=MaterialFunctionCall(X=@n53, Y=@n54) {materialFunction=&MakeFloat2}.Result
    Rotation=ScalarParameter() {defaultValue=0.75, parameterName=UVRotation}
    Scale
    Pivot
   ) {materialFunction=/Game/Global/MaterialFunctions/MF_UV_Transform}.Result) {name=ModifiedUV, nodeColor=<rgba 0 1 0.4609 1>}
  n55 = Subtract(A=NamedRerouteUsage()) {constB=0.5}
  n56 = ScalarParameter() {defaultValue=8, parameterName=NoiseTiling}
  n57 = TextureSample(UVs=Add(A=Multiply(A=@n55, B=@n56) {constB=4}) {constB=0.6}) {samplerType=SAMPLERTYPE_Grayscale, texture=&LowResBlurredNoise}
  n58 = SmoothStep(
   Min=ScalarParameter() {defaultValue=0.33, parameterName=MouthDistortionSmoothStepMin}
   Max=ScalarParameter() {defaultValue=1, parameterName=MouthDistortionSmoothStepMax}
   Value=NamedRerouteUsage()
  ) {constMin=0.5}
  NamedRerouteDeclaration(None=Multiply(A=ConstantBiasScale(None=@n57.R) {bias=-0.5, scale=2}, B=Multiply(A=@n58, B=@n25)) {constB=0.1}) {name=Noise, nodeColor=<rgba 0 0.2734 1 1>}
  NamedRerouteDeclaration(None=ScalarParameter() {defaultValue=10, parameterName=SDFMultiplier}) {name=SDFMultiplier, nodeColor=<rgba 0.6172 0 1 1>}
  n59 = NamedRerouteUsage()
  n60 = Constant() {r=0.5}
  n61 = ScalarParameter() {defaultValue=0.2, parameterName=Offset}
  n62 = Constant() {r=0.5}
  n63 = Subtract(
   A=NamedRerouteUsage()
   B=MaterialFunctionCall(X=Subtract(A=@n60, B=@n61), Y=@n62) {materialFunction=&MakeFloat2}.Result
  ) {constA=0.5}
  NamedRerouteDeclaration(None=Distance(A=Multiply(A=@n59, B=@n63), B=Constant())) {name=BaseSDF_02, nodeColor=<rgba 0 0.1094 1 1>}
  n64 = NamedRerouteUsage()
  NamedRerouteDeclaration(None=Step(X=ComponentMask(None=@n64) {r=true}) {constY=0.5}) {name=HorizontalMask, nodeColor=<rgba 1 0 0.4688 1>}
  n65 = NamedRerouteUsage()
  n66 = Subtract(
   A=NamedRerouteUsage()
   B=MaterialFunctionCall(X=Add(A=@n60, B=@n61), Y=@n62) {materialFunction=&MakeFloat2}.Result
  ) {constA=0.5}
  NamedRerouteDeclaration(None=Distance(A=Multiply(A=@n65, B=@n66), B=Constant())) {name=BaseSDF_01, nodeColor=<rgba 0 0.1094 1 1>}
  n67 = ComponentMask(None=NamedRerouteUsage()) {g=true}
  n68 = Abs(None=Multiply(A=ConstantBiasScale(None=@n67) {bias=-0.5, scale=2}, B=ScalarParameter() {defaultValue=4.5, parameterName=LineThicknessGradient_Mult}) {constB=4})
  NamedRerouteDeclaration(None=Power(Base=Saturate(None=@n68)) {constExponent=1.4}) {name=LineThicknessMask, nodeColor=<rgba 1 0 0.4688 1>}
  NamedRerouteDeclaration(None=FunctionInput() {inputName=BaseColor}) {name=InBaseColor, nodeColor=<rgba 1 0.5859 0 1>}
  n69 = ScalarParameter() {defaultValue=0.66, parameterName=LineThicknessMax2}
  n70 = ScalarParameter() {defaultValue=0.52, parameterName=TopTaperPosition}
  n71 = ScalarParameter() {defaultValue=0.05, parameterName=TopTaperWidth}
  n72 = TextureCoordinate()
  n73 = Multiply(
   A=SmoothStep(Min=@n70, Max=Add(A=@n70, B=@n71), Value=ComponentMask(None=@n72) {r=true}) {constMax=0.55, constMin=0.5}
   B=ScalarParameter() {defaultValue=4, parameterName=TopTaperIntensity}
  )
  MaterialExpressionReroute_23 = Reroute(None=@n73)
  n74 = ScalarParameter() {defaultValue=16, parameterName=ThicknessSDFMask_Power}
  n75 = Saturate(None=Add(A=NamedRerouteUsage(), B=Power(Base=@n1.R, Exp=@n74) {constExponent=4}))
  n76 = OneMinus(None=NamedRerouteUsage())
  NamedRerouteDeclaration(None=LinearInterpolate(A=Add(A=@n69, B=@MaterialExpressionReroute_23.Output), B=ScalarParameter() {parameterName=LineThicknessMin2}, Alpha=Multiply(A=@n75, B=@n76))) {name=LineThickness, nodeColor=<rgba 0 1 0.3438 1>}
  NamedRerouteDeclaration(None=@n11) {name=MouthGradient, nodeColor=<rgba 0 1 0.3906 1>}
  n77 = NamedRerouteUsage()
  n78 = Constant3Vector() {constant=<rgba 0 5 162 0>}
  n79 = Constant2Vector() {g=0.5, r=0.5}
  n80 = Frac(None=Multiply(A=Time()) {constB=0.33})
  n81 = Multiply(A=Power(Base=Multiply(A=@n80) {constB=2.6})) {constB=0.25}
  n82 = Multiply(A=@n81) {constB=0.5}
  n83 = Multiply(
   A=SphereMask(A=@n21.UVs, B=@n79, Radius=@n81) {attenuationRadius=0.5, hardnessPercent=50}
   B=OneMinus(None=SphereMask(A=@n21.UVs, B=@n79, Radius=@n82) {attenuationRadius=0.5, hardnessPercent=50})
  )
  n84 = Multiply(A=Multiply(A=SphereMask(A=@n77, B=@n78) {attenuationRadius=15, hardnessPercent=70}, B=Saturate(None=@n24)), B=@n83)
  n85 = ScalarParameter() {defaultValue=0.5, parameterName=PulseIntensity_Base}
  NamedRerouteDeclaration(None=Multiply(A=Multiply(A=@n84, B=@n85) {constB=0.25}, B=NamedRerouteUsage())) {name=WavesAlpha, nodeColor=<rgba 1 0.3984 0 1>}
  NamedRerouteDeclaration(None=Multiply(A=Constant3Vector() {constant=<rgba 0.199 0.7925 0.8212 1>}, B=Add(A=@n83))) {name=Waves, nodeColor=<rgba 1 0 0.2109 1>}
  NamedRerouteDeclaration(None=@n2) {name=StaticLinesSDFModified, nodeColor=<rgba 0.5 1 0 1>}
  NamedRerouteDeclaration(None=@n1.R) {name=StaticLinesSDF, nodeColor=<rgba 0.8516 0 1 1>}
  n86 = AppendVector(A=Constant() {r=0.5}, B=Constant() {r=1})
  n87 = Constant() {r=0.3}
  n88 = Constant() {r=0.5}
  NamedRerouteDeclaration(None=SphereMask(
    A=MaterialFunctionCall(UVs=@n7, Texture Scale=@n86) {materialFunction=&ScaleUVsByCenter}.UVs
    B=AppendVector(A=@n87, B=@n88)
    Radius=ScalarParameter() {defaultValue=0.15, parameterName=MouthGradient2Radius}
    Hardness=ScalarParameter() {defaultValue=0.3, parameterName=MouthGradient2Hardness}
   ) {attenuationRadius=0.1, hardnessPercent=0}) {name=MouthGradient2, nodeColor=<rgba 1 0 0.6094 1>}
  n89 = SmoothStep(Value=OneMinus(None=@n19)) {constMax=0.6, constMin=0.25}
  MaterialExpressionReroute_21 = Reroute(None=@n89)
  NamedRerouteDeclaration(None=@MaterialExpressionReroute_21.Output) {name=EarGradient, nodeColor=<rgba 0.2891 0 1 1>}
  NamedRerouteDeclaration(None=Max(A=@n30, B=@n29.G)) {name=StaticAndDynamicLines, nodeColor=<rgba 0 1 0.4375 1>}
  PinBase()
  PinBase!(Output=@n73) {reroutePins [Output {expression=&self:MaterialExpressionReroute_23}]}
  PinBase!() {reroutePins [
    A {expression=&self:MaterialExpressionReroute_17}
    B {expression=&self:MaterialExpressionReroute_10}
    K {expression=&self:MaterialExpressionReroute_18}
   ]}
  PinBase!(Output=@n51) {reroutePins [Output {expression=&self:MaterialExpressionReroute_16}]}
  PinBase!()
  PinBase!(Output=@n89, Output2=@n20) {reroutePins [
    Output {expression=&self:MaterialExpressionReroute_21}
    Output2 {expression=&self:MaterialExpressionReroute_22}
   ]}
 ]}
